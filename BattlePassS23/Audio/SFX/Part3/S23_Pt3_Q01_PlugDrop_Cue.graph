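# Sound cue for a plug drop effect in Battle Pass S23 Part 3 quest 01, routed to the Fort_Quests sound class.
# A Random node picks one of three equally weighted wave variants and feeds a Modulator at default settings as the root.
# Spatialization uses an attenuation asset named for Q04 rather than Q01, with a limited max distance.
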SoundCue /BattlePassS23/Audio/SFX/Part3/S23_Pt3_Q01_PlugDrop_Cue
root=Modulator_0
nodes [
 {id=Modulator_0, children [Random_0]}
 {id=Random_0, children [WavePlayer_0,WavePlayer_1,WavePlayer_2], properties {weights [1,1,1]}}
 {id=WavePlayer_0, properties {soundWaveAssetPtr=&/S23_Pt3_Q01_PlugDrop_01}}
 {id=WavePlayer_1, properties {soundWaveAssetPtr=&/S23_Pt3_Q01_PlugDrop_02}}
 {id=WavePlayer_2, properties {soundWaveAssetPtr=&/S23_Pt3_Q01_PlugDrop_03}}
]
properties {
 attenuationSettings=&/S23_Pt03_Q04_Atten
 maxDistance=1100
 soundClassObject=&Fort_Quests
}
