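# Sound cue for a looping socket power sound effect in the Season 23 Battle Pass Part 3 quest audio.
# A single looping wave player feeds a Modulator node left at default settings.
# It plays under the Fort_Quests sound class and borrows the attenuation asset named for Q04, although the cue itself is named for Q01.
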SoundCue /BattlePassS23/Audio/SFX/Part3/S23_Pt3_Q01_Socket_Power_LP_Cue
root=Modulator_0
nodes [
 {id=Modulator_0, children [WavePlayer_0]}
 {id=WavePlayer_0, properties {bLooping=true, soundWaveAssetPtr=&/S23_Pt3_Q01_Socket_Power_LP}}
]
properties {
 attenuationSettings=&/S23_Pt03_Q04_Atten
 maxDistance=1100
 soundClassObject=&Fort_Quests
 volumeMultiplier=1
}
